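# Niagara system for the Color Block backpack cosmetic that emits small color-shifting lights attached to the skeletal mesh.
# A single light-renderer emitter continuously spawns short-lived particles in two spawn groups, using the group as bone/socket index on User.SkeletalMesh.
# Each particle cycles through red, green and blue over its life, with RGB scaled by time of day.
# FN_Light_Attributes takes that color and a rise-and-fall curve, making each light pulse.
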NiagaraSystem /BRCosmetics/Accessories/FORT_Backpacks/M_MED_Color_Block/FX/NS_Backpack_Color_Block_Light
dependencies {
 usedDataInterfaces [&NiagaraDataInterfaceColorCurve, &NiagaraDataInterfaceCurve]
 usedDynamicInputs []
 usedModules [&SystemState]
}
emitters {Empty {
  emitterUpdateScript [
   &EmitterState()
   &SpawnRate(SpawnRate=2, Use Spawn Probability=false)
  ]
  particleSpawnScript [
   &InitializeParticle(
    Color=<rgba 40 40 40 1>
    Lifetime=1.4
    Sprite Size Mode=::NewEnumerator3
    Uniform Sprite Size=20
    UsePositionOffset=false
    Write Material Random=true
   )
   /SetVariables(Particles.LightRadius=8) {isSetParameters=true}
   /SetVariables(Particles.LightExponent=1) {isSetParameters=true}
   &FN_SetSpawnGroup(Number of Spawn Groups=2)
  ]
  particleUpdateScript [
   &ParticleState()
   &SkeletalMeshLocation(
    Bone / Socket Index=$Particles.SpawnGroup
    MaskSpawnGroup=false
    Mesh Sampling Type=::NewEnumerator5
    Skeletal Mesh=$User.SkeletalMesh
    Socket Sampling Mode=::NewEnumerator5
   )
   &ScaleColor(
    Linear Color Curve=DataInterface() {propertyValues {
      AlphaCurve {keys [<at 0 0.01>,<at 0.0807 1>,<at 0.6373 1>,<at 1 0>]}
      BlueCurve {keys [<at 0.0636 0>,<at 0.4346 0>,<at 0.73 1>,<at 0.9945 0>]}
      GreenCurve {keys [<at 0.0636 0>,<at 0.4346 1>,<at 0.73 0.0693>,<at 0.9945 0>]}
      RedCurve {keys [<at 0.0636 1>,<at 0.4346 0.103>,<at 0.73 0>,<at 0.9945 1>]}
     }}
    Scale Mode=::NewEnumerator2
   )
   &ScaleColor(
    Scale RGB=&VectorFromFloat(Value=&FNTimeOfDayScalar(Night Scale=1, Morning Scale=0.6, Evening Scale=0.6))
    ScaleA=true
    ScaleRGB=true
   ) {module=ScaleColor001}
   &FN_Light_Attributes(
    Color=$Particles.Color
    Create Color Attribute=true
    Scale Color=&FloatFromCurve(FloatCurve=DataInterface() {propertyValues {Curve {keys [
         {interpMode=RCIM_Cubic, leaveTangent=6.6616, tangentMode=RCTM_Break, time=0, value=0}
         {interpMode=RCIM_Cubic, tangentMode=RCTM_Break, time=0.38, value=1}
         {arriveTangent=-2.29, interpMode=RCIM_Cubic, tangentMode=RCTM_Break, time=1, value=0}
        ]}}})
   )
  ]
  rendererClasses [&NiagaraLightRendererProperties]
  renderers [{rendererClass=&NiagaraLightRendererProperties, rendererIndex=0}]
 }}
summary {
 systemName=NS_Backpack_Color_Block_Light
 emitters [{emitterName=Empty, rendererClasses [&NiagaraLightRendererProperties]}]
 userVariables [User.SkeletalMesh {
   type=NiagaraDataInterfaceSkeletalMesh
   defaultValue {
    struct=NiagaraExt_VariableValue_DataInterface
    value {dataInterface=&self:NiagaraDataInterfaceSkeletalMesh_0}
   }
  }]
}
