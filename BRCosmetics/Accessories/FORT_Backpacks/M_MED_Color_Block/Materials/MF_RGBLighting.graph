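# Material function that layers animated RGB fan lighting onto the Color Block backpack's material attributes, blended by an Alpha input.
# A gradient texture with tiling, panning and rotation supplies color, scaled by emissive intensity.
# The emissive chain multiplies in optional Fresnel, with named reroutes carrying the gradient, LED mask and time-rotated fan blade mask.
# The lit layer overrides base color with black, plus emissive and roughness, before blending with the input.
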MaterialFunction /BRCosmetics/Accessories/FORT_Backpacks/M_MED_Color_Block/Materials/MF_RGBLighting
graph {nodes [
  n1 = FunctionInput() {inputName=Material, inputType=FunctionInput_MaterialAttributes}
  n2 = Constant3Vector()
  n3 = NamedRerouteUsage()
  n4 = ScalarParameter() {defaultValue=0.1, group=RGB, parameterName=RGB_Roughness}
  FunctionOutput(None=BlendMaterialAttributes(
    A=@n1
    B=SetMaterialAttributes!(MaterialAttributes=@n1, BaseColor=@n2, EmissiveColor=@n3, Roughness=@n4) {attributeSetTypes [#1, #2, #3]}
    Alpha=FunctionInput() {inputName=Alpha, inputType=FunctionInput_Scalar, sortPriority=1}
   ))
  n5 = ComponentMask(None=VectorParameter() {group=RGB, parameterName=GradientUVPanSpeed, sortPriority=12, defaultValue=<rgba 0 0 0 1>}.RGB) {g=true, r=true}
  n6 = Time()
  n7 = VectorParameter() {group=RGB, parameterName=GradientUVTile, sortPriority=10, defaultValue=<rgba 1 1 0 1>}
  n8 = Add(
   A=Multiply(A=TextureCoordinate(), B=ComponentMask(None=@n7.RGB) {g=true, r=true})
   B=ComponentMask(None=VectorParameter() {group=RGB, parameterName=GradientUVOffset, sortPriority=11, defaultValue=<rgba 0 0 0 1>}.RGB) {g=true, r=true}
  )
  n9 = VectorParameter() {group=RGB, parameterName=GradientUVRotationCenterOffset, sortPriority=13, defaultValue=<rgba 0.5 0.5 0 1>}
  n10 = Add(
   A=Multiply(A=Time(), B=ScalarParameter() {group=RGB, parameterName=GradientUVRotationSpeed, sortPriority=15})
   B=ScalarParameter() {group=RGB, parameterName=GradientUVRotation, sortPriority=14}
  )
  n11 = TextureSampleParameter2D(UVs=Add(
    A=Multiply(A=@n5, B=@n6)
    B=MaterialFunctionCall(UVs=@n8, Rotation Center=@n9.RGB, "Rotation Angle (0-1)"=@n10) {materialFunction=&CustomRotator}.Rotated Values
   )) {
   group=RGB
   parameterName=GradientMap
   sortPriority=1
   texture=/BRCosmetics/Accessories/FORT_Backpacks/M_MED_Color_Block/Textures/T_M_MED_Color_Block_Backpack_Gradient_Linear
  }
  NamedRerouteDeclaration(None=Multiply(A=@n11.RGB, B=ScalarParameter() {defaultValue=1, group=RGB, parameterName=EmissiveIntensity, sortPriority=0})) {name=Gradient, nodeColor=<rgba 1 0 0.8438 1>}
  n12 = TextureCoordinate()
  n13 = ComponentMask(None=VectorParameter() {group=RGB, parameterName=LED_UV_Tiling, defaultValue=<rgba 1 1 0 1>}.RGB) {g=true, r=true}
  n14 = VectorParameter() {group=RGB, parameterName=LED_UV_Offset, sortPriority=33, defaultValue=<rgba 0 0 0 1>}
  n15 = TextureSampleParameter2D(UVs=Add(A=Multiply(A=@n12, B=@n13), B=ComponentMask(None=@n14.RGB) {g=true, r=true})) {
   group=RGB
   parameterName=LED_Mask
   samplerType=SAMPLERTYPE_LinearGrayscale
   sortPriority=30
   texture=/BRCosmetics/Accessories/FORT_Backpacks/M_MED_Color_Block/Textures/T_M_MED_Color_Block_Backpack_Fan_LED_Mask
  }
  NamedRerouteDeclaration(None=LinearInterpolate(B=@n15.R, Alpha=ScalarParameter() {defaultValue=1, group=RGB, parameterName=LED_Strength, sortPriority=31}) {constA=1}) {name=LED, nodeColor=<rgba 0.8047 1 0 1>}
  n16 = VectorParameter() {group=RGB, parameterName=FanBladeUVRotationCenterOffset, sortPriority=52, defaultValue=<rgba 0.5 0.5 0 1>}
  n17 = Time()
  n18 = ScalarParameter() {defaultValue=0.5, group=RGB, parameterName=FanBladeRotationSpeed, sortPriority=53}
  n19 = TextureSampleParameter2D(UVs=MaterialFunctionCall(UVs=TextureCoordinate(), Rotation Center=@n16.RGB, "Rotation Angle (0-1)"=Multiply(A=@n17, B=@n18)) {materialFunction=&CustomRotator}.Rotated Values) {
   parameterName=FanBladeMask
   samplerType=SAMPLERTYPE_LinearColor
   sortPriority=50
   texture=/BRCosmetics/Accessories/FORT_Backpacks/M_MED_Color_Block/Textures/T_M_MED_Color_Block_Backpack_Fan_Mask
  }
  NamedRerouteDeclaration(None=LinearInterpolate(B=@n19.R, Alpha=ScalarParameter() {defaultValue=1, group=RGB, parameterName=FanBladeStrength, sortPriority=51}) {constA=1}) {name=FanBlade, nodeColor=<rgba 0 0.9062 1 1>}
  n20 = ScalarParameter() {defaultValue=2, group=RGB, parameterName=FresnelExponent, sortPriority=61}
  n21 = Multiply(
   A=NamedRerouteUsage()
   B=LinearInterpolate(B=Fresnel(ExponentIn=@n20), Alpha=ScalarParameter() {group=RGB, parameterName=UseFresnel, sortPriority=60}) {constA=1}
  )
  n22 = NamedRerouteUsage()
  NamedRerouteDeclaration(None=Multiply(A=Multiply(A=@n21, B=@n22), B=NamedRerouteUsage())) {name=RGB_Emissive, nodeColor=<rgba 0 1 0.625 1>}
 ]}
