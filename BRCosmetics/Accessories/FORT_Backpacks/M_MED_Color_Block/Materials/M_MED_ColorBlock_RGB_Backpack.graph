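# Character-style material for the Color Block back bling cosmetic, adding RGB lighting and animated hue-shifting glow.
# The shared PreFX output feeds MF_RGBLighting, with channel-selectable FX texture masks exposed for RGB (optional) and hue shift.
# A time-driven HueShift drives emissive over black base color, blended with the unshifted material by mask.
# The blend then passes through the shared PostFX and cinematic post functions.
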Material /BRCosmetics/Accessories/FORT_Backpacks/M_MED_Color_Block/Materials/M_MED_ColorBlock_RGB_Backpack
graph {
 properties {
  bUseMaterialAttributes=true
  bUsedWithMeshDeformer=true
  bUsedWithSkeletalMesh=true
  pixelDepthOffsetMode=PDOM_Legacy
  usedShadingModels=Default Lit
 }
 nodes [
  n1 = MaterialFunctionCall() {materialFunction=&BaseFN_CharacterFunction_PreFX}
  n2 = NamedRerouteUsage()
  n3 = Constant3Vector()
  n4 = Time()
  n5 = ScalarParameter() {group=RGB, parameterName=HueShiftSpeed, sortPriority=81}
  n6 = Multiply(
   A=MaterialFunctionCall(Hue Shift Percentage=Multiply(A=@n4, B=@n5) {constB=0.5}, Texture=NamedRerouteUsage()) {materialFunction=&HueShift}.Result
   B=ScalarParameter() {defaultValue=1, group=RGB, parameterName=HueShiftEmissiveStrength, sortPriority=82}
  )
  n7 = MaterialFunctionCall(
   AO=@n1.AO
   Distance Blend=@n1.AO
   Fresnel=@n1.AO
   Gmap=@n1.AO
   Pre Skinned Local Position=@n1.AO
   Material=BlendMaterialAttributes(
    A=@n2
    B=SetMaterialAttributes!(MaterialAttributes=@n2, Base Color=@n3, Emissive Color=@n6) {attributeSetTypes [#1, #2]}
    Alpha=NamedRerouteUsage()
   )
  ) {materialFunction=&BaseFN_CharacterFunction_PostFX}
  MaterialFunctionCall(In=@n7.Result) {materialFunction=&CINE_MF_CharPostFX}
  n8 = TextureSample(Tex=TextureObjectParameter() {group=Skin FX, parameterName=SkinFX_Mask, texture=&T_BlacksRGB}) {texture=/BRCosmetics/Accessories/FORT_Backpacks/M_MED_Color_Block/Textures/T_M_MED_Color_Block_Backpack_FX}
  NamedRerouteDeclaration(None=StaticSwitchParameter(True=ChannelMaskParameter(None=@n8.RGBA) {group=RGB, parameterName=RGB_Channel, sortPriority=-9}.R, False=Constant() {r=1}) {group=RGB, parameterName=UseRGBMask, sortPriority=-10}) {name=RGB_Mask, nodeColor=<rgba 0.1016 1 0 1>}
  n9 = ChannelMaskParameter(None=@n8.RGBA) {group=RGB, maskChannel=Green, parameterName=HueShift_Channel, sortPriority=80, defaultValue=<rgba 0 1 0 0>}
  NamedRerouteDeclaration(None=@n9.R) {name=HueShiftMask, nodeColor=<rgba 1 0.2344 0 1>}
  n10 = GetMaterialAttributes(None=MaterialFunctionCall!(Material=@n1.Result, Alpha=NamedRerouteUsage()) {materialFunction=&/MF_RGBLighting}.Result) {attributeGetTypes [#1]}
  NamedRerouteDeclaration(None=@n10.Base Color) {name=BaseColor, nodeColor=<rgba 0 0.8828 1 1>}
  NamedRerouteDeclaration(None=@n10.MaterialAttributes) {name=RGB_Material, nodeColor=<rgba 1 0.7031 0 1>}
 ]
}
