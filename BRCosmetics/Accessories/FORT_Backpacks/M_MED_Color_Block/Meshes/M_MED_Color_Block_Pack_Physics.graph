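# Physics asset for the M_MED_Color_Block backpack cosmetic, giving its dangling cable bones simulated collision and motion.
# The pack bone carries a capsule and box body, while the master cable bone has a small sphere and the front and back cable bones have thin capsules.
# Constraints form a short branching chain, with both cables jointed to the master and the master to the pack, each with narrow swing and twist limits.
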PhysicsAsset /BRCosmetics/Accessories/FORT_Backpacks/M_MED_Color_Block/Meshes/M_MED_Color_Block_Pack_Physics
bodies {
 dyn_cable_MSTR [{shapeType=Sphere, radius=0.505, shapeName=dyn_cable_MSTR_sphere}]
 dyn_cable_b [{
   shapeType=Capsule
   length=4.15
   radius=0.5
   shapeName=dyn_cable_b_capsule
   center=<1.4451 0.0151 -0.1616>
   rotation=<rot 0.1676 -92.5899 -1.9396>
  }]
 dyn_cable_f [{
   shapeType=Capsule
   length=4.1371
   radius=0.4
   shapeName=dyn_cable_f_capsule
   center=<1.1224 0.1285 0.0683>
   rotation=<rot 1.8068 -89.9993 -4.4105>
  }]
 pack [
  {shapeType=Capsule, length=6.5, radius=3.5, shapeName=pack_capsule, center=<7.0379 -3.7467 0.6427>, rotation=<rot 90 -0.0098 -0.0098>}
  {shapeType=Box, extentX=11, extentY=4.5, extentZ=7, shapeName=pack_box, center=<6.7094 -5.5523 -4.1766>}
 ]
}
constraints [
 {bone1Name=dyn_cable_f, bone2Name=dyn_cable_MSTR, swing1LimitDegrees=30, swing2LimitDegrees=15, twistLimitDegrees=3}
 {bone1Name=dyn_cable_b, bone2Name=dyn_cable_MSTR, swing1LimitDegrees=16, swing2LimitDegrees=6, twistLimitDegrees=4}
 {bone1Name=dyn_cable_MSTR, bone2Name=pack, swing1LimitDegrees=8, swing2LimitDegrees=20, twistLimitDegrees=4}
]
